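; Vehicle spawner Blueprint for the Hoagie vehicle that adjusts each vehicle's collision as it spawns.
; On vehicle spawned, it calls the parent handler, checks validity and casts the vehicle to HoagieVehicle.
; It then sets the vehicle's overlap collision box to query-only collision and enables overlap events on it.
; The construction script only calls its parent, and BeginPlay holds just a disabled placeholder node.
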(event EventBeginPlay
  (|CallParentFunction) ; This node is disabled and will not be called.\nDrag off pins to build functionality.
)

(event EventOnVehicleSpawned
  (params (Spawned%20Vehicle :type Fort%20Athena%20Vehicle%20Object%20Reference))
  ;[0.1448 0.2029 1] Set Collision box to generat overlap events
  (|CallParentFunction Spawned%20Vehicle)
  (Utilities|IsValid Spawned%20Vehicle
    (:Is%20Valid
      (bind _ashoagie_vehicle (Utilities|Casting|CastToHoagieVehicle Spawned%20Vehicle))
      (Collision|SetCollisionEnabled (Variables|/Hoagie/HoagieVehicle.HoagieVehicle_C|GetOverlapCollisionBox _ashoagie_vehicle) "QueryOnly")
      (Variables|/Script/Engine.PrimitiveComponent|SetbGenerateOverlapEvents true (Variables|/Hoagie/HoagieVehicle.HoagieVehicle_C|GetOverlapCollisionBox _ashoagie_vehicle)))))

(fn ConstructionScript
  (|CallParentFunction))
