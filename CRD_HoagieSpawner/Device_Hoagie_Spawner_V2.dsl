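; Fortnite Creative spawner device for the Hoagie vehicle, exposing user options for fuel, wraps and skydive on exit.
; On spawn it casts a valid vehicle to HoagieVehicle, makes its overlap box query-only with overlap events, and applies the skydive-on-exit option.
; Getter and setter functions map fuel and wrap settings to user option variables.
; The min and max random starting fuel setters also call the parent implementation.
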(event EventBeginPlay
  (|CallParentFunction) ; This node is disabled and will not be called.\nDrag off pins to build functionality.
)

(event EventOnVehicleSpawned
  (params (Spawned%20Vehicle :type Fort%20Athena%20Vehicle%20Object%20Reference))
  ;[0.1448 0.2029 1] Set Collision box to generat overlap events
  (|CallParentFunction Spawned%20Vehicle)
  (Utilities|IsValid Spawned%20Vehicle
    (:Is%20Valid
      (bind _ashoagie_vehicle (Utilities|Casting|CastToHoagieVehicle Spawned%20Vehicle))
      (Collision|SetCollisionEnabled (Variables|/Hoagie/HoagieVehicle.HoagieVehicle_C|GetOverlapCollisionBox _ashoagie_vehicle) "QueryOnly")
      (Variables|/Script/Engine.PrimitiveComponent|SetbGenerateOverlapEvents true (Variables|/Hoagie/HoagieVehicle.HoagieVehicle_C|GetOverlapCollisionBox _ashoagie_vehicle))
      (Variables|/Script/FortniteGame.FortAthenaVehicle|SetbSkydivePlayerOnVehicleExit (Variables|UserOptions|Advanced|GetSkydive%20on%20Exit) _ashoagie_vehicle))))

(fn GetSupportsWraps
  (return (Variables|UserOptions|Basic|GetSupportsWraps)))

(fn GetConsumesFuel
  (return (Variables|UserOptions|GetFuel%20Consumption)))

(fn GetFuelUseMultiplier
  (return (Variables|UserOptions|GetFuelUseMultiplier)))

(fn GetMaxRandomStartingFuel
  (return (Variables|UserOptions|Advanced|GetMax%20Random%20Starting%20Fuel)))

(fn GetMinRandomStartingFuel
  (return (Variables|UserOptions|Advanced|GetMin%20Random%20Starting%20Fuel)))

(fn GetRandomStartingFuel
  (return (Variables|UserOptions|GetRandom%20Starting%20Fuel)))

(fn GetStartingFuelValue
  (return (Variables|UserOptions|GetStarting%20Fuel)))

(fn SetConsumeFuel
  (params (ConsumeFuel :type Boolean))
  (Variables|UserOptions|SetFuel%20Consumption ConsumeFuel))

(fn SetRandomStartingFuel
  (params (RandomStartingFuel :type Boolean))
  (Variables|UserOptions|SetRandom%20Starting%20Fuel RandomStartingFuel))

(fn SetStartingFuelValue
  (params (Starting%20Fuel%20Value :type Integer))
  (Variables|UserOptions|SetStarting%20Fuel Starting%20Fuel%20Value))

(fn SetSupportsWraps
  (params (Supports%20Wraps :type Boolean))
  (Variables|UserOptions|Basic|SetSupportsWraps Supports%20Wraps))

(fn SetMaxRandomStartingFuel
  (params (Max%20Fuel%20Percent :type Integer))
  (|CallParentFunction Max%20Fuel%20Percent)
  (Variables|UserOptions|Advanced|SetMax%20Random%20Starting%20Fuel Max%20Fuel%20Percent))

(fn SetMinRandomStartingFuel
  (params (Min%20Fuel%20Percent :type Integer))
  (|CallParentFunction Min%20Fuel%20Percent)
  (Variables|UserOptions|Advanced|SetMin%20Random%20Starting%20Fuel Min%20Fuel%20Percent))

(fn ConstructionScript
  (|CallParentFunction))
